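# Physics asset for the fur mesh of the medium female Barbarian player body, covering fur on both calves and forearms.
# Each dynamic fur bone and its calf or lower-arm twist bone carries a single sphere body.
# Four constraints tie each fur bone to its twist bone, with both swings and twist locked, so the configured angular limits allow no relative rotation.
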PhysicsAsset /Game/Characters/Player/Female/Medium/Bodies/F_MED_Barbarian/Meshes/F_MED_Barbarian_Fur_PHYSICS
bodies {
 calf_twist_02_l [{shapeType=Sphere, radius=9.8499, center=<-5.5509 -1.3139 -0.5136>}]
 calf_twist_02_r [{shapeType=Sphere, radius=11.3504, center=<3.7028 1.6301 1.1426>}]
 dyn_calfFur_l [{shapeType=Sphere, radius=9.4104}]
 dyn_calfFur_r [{shapeType=Sphere, radius=9.5174}]
 dyn_forearmFur_l [{shapeType=Sphere, radius=5.1166}]
 dyn_forearmFur_r [{shapeType=Sphere, radius=5.1401}]
 lowerarm_twist_02_l [{shapeType=Sphere, radius=7.0343}]
 lowerarm_twist_02_r [{shapeType=Sphere, radius=7.748, center=<2.7865 0.8693 -0.8345>}]
}
constraints [
 {
  bone1Name=dyn_calfFur_r
  bone2Name=calf_twist_02_r
  swing1LimitDegrees=45
  swing1Motion=Locked
  swing2LimitDegrees=45
  swing2Motion=Locked
  twistLimitDegrees=45
  twistMotion=Locked
 }
 {
  bone1Name=dyn_calfFur_l
  bone2Name=calf_twist_02_l
  swing1LimitDegrees=45
  swing1Motion=Locked
  swing2LimitDegrees=45
  swing2Motion=Locked
  twistLimitDegrees=45
  twistMotion=Locked
 }
 {
  bone1Name=dyn_forearmFur_l
  bone2Name=lowerarm_twist_02_l
  swing1LimitDegrees=45
  swing1Motion=Locked
  swing2LimitDegrees=45
  swing2Motion=Locked
  twistLimitDegrees=45
  twistMotion=Locked
 }
 {
  bone1Name=dyn_forearmFur_r
  bone2Name=lowerarm_twist_02_r
  swing1LimitDegrees=45
  swing1Motion=Locked
  swing2LimitDegrees=45
  swing2Motion=Locked
  twistLimitDegrees=45
  twistMotion=Locked
 }
]
