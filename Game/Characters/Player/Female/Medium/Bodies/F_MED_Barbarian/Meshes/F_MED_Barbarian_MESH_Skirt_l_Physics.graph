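# Physics asset for the left skirt mesh of the medium female Barbarian player character, defining bodies for its back, front and side flaps.
# Each flap's dynamic bone has a single box body, while the left upper thigh twist bone has two overlapping capsules approximating the leg.
# All three flaps constrain directly to that thigh bone rather than to each other, with moderate swing limits and tight twist limits.
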PhysicsAsset /Game/Characters/Player/Female/Medium/Bodies/F_MED_Barbarian/Meshes/F_MED_Barbarian_MESH_Skirt_l_Physics
bodies {
 dyn_skirtBack_l [{shapeType=Box, extentX=4.01, extentY=17.5418, extentZ=18.5465, center=<8.0534 0.0003 0>, rotation=<rot -5.5127 89.9999 89.9999>}]
 dyn_skirtFront_l [{shapeType=Box, extentX=3.07, extentY=18.1436, extentZ=16.5425, center=<8.0258 -0.1396 0.0251>, rotation=<rot -10.2119 90.8979 89.8406>}]
 dyn_skirtSide_l [{shapeType=Box, extentX=2.9529, extentY=12.8684, extentZ=19.7931, center=<8.6255 3.0949 0.0004>, rotation=<rot 0 99.4586 89.9995>}]
 thigh_twist_01_l [
  {shapeType=Capsule, length=29.4715, radius=6.9158, center=<7.769 -0.0001 0>, rotation=<rot 0 90 90>}
  {shapeType=Capsule, length=17.5309, radius=6.5121, center=<9.6206 2.1015 -0.1197>, rotation=<rot 0 87.01 87.505>}
 ]
}
constraints [
 {bone1Name=dyn_skirtBack_l, bone2Name=thigh_twist_01_l, swing1LimitDegrees=20, swing2LimitDegrees=20, twistLimitDegrees=2}
 {bone1Name=dyn_skirtFront_l, bone2Name=thigh_twist_01_l, swing1LimitDegrees=20, swing2LimitDegrees=20, twistLimitDegrees=5}
 {bone1Name=dyn_skirtSide_l, bone2Name=thigh_twist_01_l, swing1LimitDegrees=20, swing2LimitDegrees=20, twistLimitDegrees=5}
]
