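# Physics asset for the right-side skirt mesh of the medium female Barbarian player body, defining collision bodies and joints for its skirt bones.
# Box bodies cover the back, front and side skirt bones, with angled extra boxes on front and side, while a capsule represents the right upper thigh twist bone.
# Each panel is constrained to the thigh with tight angular limits, and a looser constraint links side to back.
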PhysicsAsset /Game/Characters/Player/Female/Medium/Bodies/F_MED_Barbarian/Meshes/F_MED_Barbarian_MESH_Skirt_r_Physics
bodies {
 dyn_skirtBack_r [{shapeType=Box, extentX=3.4503, extentY=13.2208, extentZ=24.5198, center=<11.255 -0.7373 0.7609>, rotation=<rot 4.9283 89.9999 89.9999>}]
 dyn_skirtFront_r [
  {shapeType=Box, extentX=1.9415, extentY=13.1009, extentZ=23.2145, center=<10.5442 -0.4865 0.5448>, rotation=<rot 3.003 89.9995 89.9996>}
  {shapeType=Box, extentX=24.8145, extentY=2.0098, extentZ=9.0598, center=<10.8868 -0.3327 -4.7805>, rotation=<rot 0 0 -25.0949>}
 ]
 dyn_skirtSide_r [
  {shapeType=Box, extentX=2.8094, extentY=22.563, extentZ=22.2747, center=<10.5748 -0.6243 0.4047>, rotation=<rot 0 90.803 89.9999>}
  {shapeType=Box, extentX=25.4512, extentY=2.9703, extentZ=15.1221, center=<13.1202 1.7924 -7.1466>, rotation=<rot 0 0 -40.3697>}
  {shapeType=Box, extentX=16.8478, extentY=2.743, extentZ=13.0355, center=<15.6202 2.55 8.4165>, rotation=<rot -3.0279 3.177 43.5829>}
 ]
 thigh_twist_01_r [{shapeType=Capsule, length=5.5902, radius=8.784, center=<-0.0001 -0.5271 -2.0704>, rotation=<rot 90 0 0>}]
}
constraints [
 {bone1Name=dyn_skirtBack_r, bone2Name=thigh_twist_01_r, swing1LimitDegrees=25, swing2LimitDegrees=25, twistLimitDegrees=10}
 {bone1Name=dyn_skirtSide_r, bone2Name=thigh_twist_01_r, swing1LimitDegrees=25, swing2LimitDegrees=25, twistLimitDegrees=10}
 {bone1Name=dyn_skirtFront_r, bone2Name=thigh_twist_01_r, swing1LimitDegrees=25, swing2LimitDegrees=25, twistLimitDegrees=10}
 {bone1Name=dyn_skirtSide_r, bone2Name=dyn_skirtBack_r, swing1LimitDegrees=30, swing2LimitDegrees=30, twistLimitDegrees=30}
]
