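# Post-process style AnimBlueprint for the female Barbarian body that adds secondary motion to skirt and fur.
# It copies the parent mesh pose, then chains RigidBody sims for fur and both skirt sides, with skirt weight driven by crouch state.
# Bone constraints, Anim Dynamics and a bunShrink scale finish the pose.
# The update event also stores state-based gravity from a deprecated function, which the AnimGraph never reads.
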AnimBlueprint /Game/Characters/Player/Female/Medium/Bodies/F_MED_Barbarian/Meshes/F_MED_Barbarian_barbarian_AnimBP
target_skeleton=&/F_MED_Barbarian_barbarian_MESH_Skeleton
graphs [
 AnimGraph {nodes [
   n1 = RigidBody(ComponentPose=Local To Component(LocalPose=Copy Pose From Mesh() {bUseAttachedParent=true}), Alpha=0.5) {
    bClampLinearTranslationLimitToRefPose=true
    bOverrideWorldGravity=true
    overridePhysicsAsset=&/F_MED_Barbarian_Fur_PHYSICS
    simulationSpace=BaseBoneSpace
    alphaBoolBlend {blendOption=Linear}
    baseBoneRef {boneName=pelvis}
    componentAppliedLinearAccClamp=<50 50 50>
    componentLinearAccScale=<0.4 0.4 0.4>
    componentLinearVelScale=<0.5 0.5 0.5>
    overrideWorldGravity {z=-980}
   }
   n2 = Get crouchAlpha() -> (crouchAlpha:double)
   n3 = RigidBody!(ComponentPose=@n1, Alpha=@n2, OverrideWorldGravity:Vector="0,0,-980", ExternalForce) -> (Pose) {
    bOverrideWorldGravity=true
    overridePhysicsAsset=&/F_MED_Barbarian_MESH_Skirt_l_Physics
    simulationSpace=BaseBoneSpace
    alphaBoolBlend {blendOption=Linear}
    baseBoneRef {boneName=pelvis}
    componentAppliedLinearAccClamp=<500 500 500>
    componentLinearAccScale=<0.7 0.7 0.7>
    componentLinearVelScale=<0.6 0.6 0.6>
    overrideWorldGravity {z=-980}
   }
   n4 = RigidBody(ComponentPose=@n3, Alpha=@n2, OverrideWorldGravity="0,0,-980") {
    bOverrideWorldGravity=true
    overridePhysicsAsset=&/F_MED_Barbarian_MESH_Skirt_r_Physics
    simulationSpace=BaseBoneSpace
    alphaBoolBlend {blendOption=Linear}
    baseBoneRef {boneName=thigh_r}
    componentAppliedLinearAccClamp=<500 500 500>
    componentLinearAccScale=<0.7 0.7 0.7>
    componentLinearVelScale=<0.6 0.6 0.6>
    overrideWorldGravity {z=-980}
   }
   n5 = "Constraint - dyn_skirtSide_r (P)"(ComponentPose=@n4, Alpha=1, ConstraintWeights_0=1) {
    alphaBoolBlend {blendOption=Linear}
    boneToModify {boneName=dyn_skirtSide_r}
    constraintSetup [{targetBone {boneName=thigh_twist_01_r}}]
    constraintWeights [1]
   }
   n6 = Anim Dynamics - Bone: dyn_skirtSide_r!(ComponentPose=@n5, Alpha=1, GravityOverride:Vector="0,0,0", ExternalForce="0,0,0") -> (Pose) {
    angularDampingOverride=0.97
    angularSpringConstant=2000
    bAngularSpring=true
    bGravityOverrideInSimSpace=true
    bOverrideAngularDamping=true
    bUseGravityOverride=true
    simulationSpace=BoneRelative
    alphaBoolBlend {blendOption=Linear}
    boundBone {boneName=dyn_skirtSide_r}
    componentAppliedLinearAccClamp=<2000 2000 2000>
    componentLinearAccScale=<0.4 0.3 0.3>
    componentLinearVelScale=<0.4 0.3 0.3>
    physicsBodyDefinitions [{
      sphereCollisionRadius=0
      boundBone {boneName=dyn_skirtSide_r}
      boxExtents=<10 20 20>
      constraintSetup {angularLimitsMax=<1 1 0>, angularLimitsMin=<-1 -1 -15>, angularTarget=<0 0 0>}
      localJointOffset=<0.5 0 0>
     }]
    relativeSpaceBone {boneName=root}
   }
   n7 = "Constraint - dyn_skirtBack_r (P)"(ComponentPose=@n6, Alpha=1, ConstraintWeights_0=1) {
    alphaBoolBlend {blendOption=Linear}
    boneToModify {boneName=dyn_skirtBack_r}
    constraintSetup [{targetBone {boneName=thigh_twist_01_r}}]
    constraintWeights [1]
   }
   n8 = Anim Dynamics - Bone: dyn_skirtBack_r!(ComponentPose=@n7, Alpha=1, GravityOverride:Vector="0,0,0", ExternalForce="0,0,0") -> (Pose) {
    angularDampingOverride=0.97
    angularSpringConstant=2000
    bAngularSpring=true
    bGravityOverrideInSimSpace=true
    bOverrideAngularDamping=true
    bUseGravityOverride=true
    simulationSpace=BoneRelative
    alphaBoolBlend {blendOption=Linear}
    boundBone {boneName=dyn_skirtBack_r}
    componentAppliedLinearAccClamp=<2000 2000 2000>
    componentLinearAccScale=<0.4 0.3 0.3>
    componentLinearVelScale=<0.4 0.3 0.3>
    physicsBodyDefinitions [{
      sphereCollisionRadius=0
      boundBone {boneName=dyn_skirtBack_r}
      boxExtents=<10 20 20>
      constraintSetup {angularLimitsMax=<1 1 0>, angularLimitsMin=<-1 -1 -15>, angularTarget=<0 0 0>}
      localJointOffset=<0.5 0 0>
     }]
    relativeSpaceBone {boneName=root}
   }
   n9 = "Constraint - dyn_skirtBack (PP)"!(ComponentPose=@n8, Alpha=1, ConstraintWeights_0=0.5, ConstraintWeights_1:float=0.5) -> (Pose) {
    alphaBoolBlend {blendOption=Linear}
    boneToModify {boneName=dyn_skirtBack}
    constraintSetup [{targetBone {boneName=dyn_skirtBack_l}}, {targetBone {boneName=dyn_skirtBack_r}}]
    constraintWeights [1,1]
   }
   n10 = "Constraint - dyn_skirtFront (PP)"!(ComponentPose=@n9, Alpha=1, ConstraintWeights_0=0.5, ConstraintWeights_1:float=0.5) -> (Pose) {
    alphaBoolBlend {blendOption=Linear}
    boneToModify {boneName=dyn_skirtFront}
    constraintSetup [{targetBone {boneName=dyn_skirtFront_l}}, {targetBone {boneName=dyn_skirtFront_r}}]
    constraintWeights [1,1]
   }
   n11 = "Constraint - furBack_r (PPP)"!(ComponentPose=@n10, Alpha=1, ConstraintWeights_0=0.7141, ConstraintWeights_1:float=0.2141, ConstraintWeights_2:float=0.0718) -> (Pose) {
    alphaBoolBlend {blendOption=Linear}
    boneToModify {boneName=furBack_r}
    constraintSetup [{targetBone {boneName=clavicle_r}}, {targetBone {boneName=upperarm_r}}, {targetBone {boneName=spine_05}}]
    constraintWeights [1,1,1]
   }
   n12 = "Constraint - furBack_l (PPP)"!(ComponentPose=@n11, Alpha=1, ConstraintWeights_0=0.7141, ConstraintWeights_1:float=0.2141, ConstraintWeights_2:float=0.0718) -> (Pose) {
    alphaBoolBlend {blendOption=Linear}
    boneToModify {boneName=furBack_l}
    constraintSetup [{targetBone {boneName=clavicle_l}}, {targetBone {boneName=upperarm_l}}, {targetBone {boneName=spine_05}}]
    constraintWeights [1,1,1]
   }
   n13 = "Transform (Modify) Bone - Bone: bunShrink"(ComponentPose=@n12, Alpha=1, Translation="0,0,0", Rotation="0,0,0", Scale="0.7,0.7,0.7") {scaleMode=BMM_Replace, alphaBoolBlend {blendOption=Linear}, boneToModify {boneName=bunShrink}, scale=<0.7 0.7 0.7>}
   n14 = Anim Dynamics - Bone: dyn_calfFur_l(ComponentPose=@n13, Alpha=0.7, ExternalForce="0,0,0") {
    angularDampingOverride=0.98
    angularSpringConstant=15000
    bAngularSpring=true
    bGravityOverrideInSimSpace=true
    bOverrideAngularDamping=true
    simulationSpace=RootRelative
    alphaBoolBlend {blendOption=Linear}
    boundBone {boneName=dyn_calfFur_l}
    componentLinearAccScale=<0.1 0.1 0.1>
    componentLinearVelScale=<0.1 0.1 0.1>
    physicsBodyDefinitions [{
      sphereCollisionRadius=0
      boundBone {boneName=dyn_calfFur_l}
      boxExtents=<10 30 30>
      constraintSetup {angularLimitsMax=<30 30 30>, angularLimitsMin=<-30 -30 -30>, angularTarget=<1 0 0>}
      localJointOffset=<0.5 0 0>
     }]
   }
   n15 = Anim Dynamics - Bone: dyn_calfFur_r(ComponentPose=@n14, Alpha=0.7, ExternalForce="0,0,0") {
    angularDampingOverride=0.98
    angularSpringConstant=15000
    bAngularSpring=true
    bGravityOverrideInSimSpace=true
    bOverrideAngularDamping=true
    simulationSpace=RootRelative
    alphaBoolBlend {blendOption=Linear}
    boundBone {boneName=dyn_calfFur_r}
    componentLinearAccScale=<0.1 0.1 0.1>
    componentLinearVelScale=<0.1 0.1 0.1>
    physicsBodyDefinitions [{
      sphereCollisionRadius=0
      boundBone {boneName=dyn_calfFur_r}
      boxExtents=<10 30 30>
      constraintSetup {angularLimitsMax=<30 30 30>, angularLimitsMin=<-30 -30 -30>, angularTarget=<1 0 0>}
      localJointOffset=<-0.5 0 0>
     }]
   }
   Output Pose(Result=Component To Local(ComponentPose=@n15)) {layerGroup=None}
  ]}
 EventGraph {nodes [
   Event BlueprintUpdateAnimation()
   > n1 = Sequence()
   n2 = "To Float (Boolean)"(InBool=Get bIsCrouching())
   > Set crouchAlpha(crouchAlpha:double=float - float(B=float * float(A=@n2, B=0.8)) {is=title K2Node_CallFunction}) -> (Output_Get:double)
   n3 = Gravity Override State Function DEPRECATED!(
    execute=@n1.then_1
    CharacterPartAnimInstance=Self-Reference()
    GameTimeInSeconds=GetGameTimeInSeconds()
    GravityOverrideParams_Skydive_GoingUp_28_#1_WindFrequency_28_#2:Vector="1.5,7,8"
    GravityOverrideParams_Skydive_GoingUp_28_#1_WindAmplitude_31_#3:Vector="200,-500,400"
    GravityOverrideParams_Skydive_GoingUp_28_#1_WindOffset_32_#4:Vector="0,0,0"
    GravityOverrideParams_Skydive_GoingUp_28_#1_GravityOverride_15_#5:Vector="0,0,-100"
    GravityOverrideParams_Skydive_GoingUp_28_#1_JointName_27_#6:Name=root
    GravityOverrideParams_Skydive_FloatingDown_29_#7_WindFrequency_28_#2:Vector="1.5,7,8"
    GravityOverrideParams_Skydive_FloatingDown_29_#7_WindAmplitude_31_#3:Vector="200,-500,400"
    GravityOverrideParams_Skydive_FloatingDown_29_#7_WindOffset_32_#4:Vector="0,-600,400"
    GravityOverrideParams_Skydive_FloatingDown_29_#7_GravityOverride_15_#5:Vector="0,0,0"
    GravityOverrideParams_Skydive_FloatingDown_29_#7_JointName_27_#6:Name=root
    GravityOverrideParams_Skydive_DivingDown_30_#8_WindFrequency_28_#2:Vector="1.5,7,8"
    GravityOverrideParams_Skydive_DivingDown_30_#8_WindAmplitude_31_#3:Vector="50,300,100"
    GravityOverrideParams_Skydive_DivingDown_30_#8_WindOffset_32_#4:Vector="0,0,0"
    GravityOverrideParams_Skydive_DivingDown_30_#8_GravityOverride_15_#5:Vector="0,0,0"
    GravityOverrideParams_Skydive_DivingDown_30_#8_JointName_27_#6:Name=pelvis
    GravityOverrideParams_Skydive_Parachuting_31_#9_WindFrequency_28_#2:Vector="1.5,7,8"
    GravityOverrideParams_Skydive_Parachuting_31_#9_WindAmplitude_31_#3:Vector="200,-500,400"
    GravityOverrideParams_Skydive_Parachuting_31_#9_WindOffset_32_#4:Vector="0,0,0"
    GravityOverrideParams_Skydive_Parachuting_31_#9_GravityOverride_15_#5:Vector="0,-800,-980"
    GravityOverrideParams_Skydive_Parachuting_31_#9_JointName_27_#6:Name=root
    GravityOverrideParams_OnGround_Standing_32_#10:Gravity Override Setting Struct Structure=(WindFrequency_28_#2=(X=0,Y=0,Z=0),WindAmplitude_31_#3=(X=0,Y=0,Z=0),WindOffset_32_#4=(X=0,Y=0,Z=0),GravityOverride_15_#5=(X=0,Y=0,Z=-980),JointName_27_#6="root")
    GravityOverrideParams_OnGround_Moving_33_#11:Gravity Override Setting Struct Structure=(WindFrequency_28_#2=(X=0,Y=0,Z=0),WindAmplitude_31_#3=(X=0,Y=0,Z=0),WindOffset_32_#4=(X=0,Y=0,Z=0),GravityOverride_15_#5=(X=0,Y=0,Z=-980),JointName_27_#6="root")
    GravityOverrideParams_EmoteOrMelee_34_#12:Gravity Override Setting Struct Structure=(WindFrequency_28_#2=(X=0,Y=0,Z=0),WindAmplitude_31_#3=(X=0,Y=0,Z=0),WindOffset_32_#4=(X=0,Y=0,Z=0),GravityOverride_15_#5=(X=0,Y=0,Z=-980),JointName_27_#6="root")
    GravityOverrideParams_Swimming_35_#13:Gravity Override Setting Struct Structure=(WindFrequency_28_#2=(X=0,Y=0,Z=0),WindAmplitude_31_#3=(X=0,Y=0,Z=0),WindOffset_32_#4=(X=0,Y=0,Z=0),GravityOverride_15_#5=(X=-800,Y=0,Z=0),JointName_27_#6="pelvis")
    GravityOverrideParams_Front_End_36_#14:Gravity Override Setting Struct Structure=(WindFrequency_28_#2=(X=0,Y=0,Z=0),WindAmplitude_31_#3=(X=0,Y=0,Z=0),WindOffset_32_#4=(X=0,Y=0,Z=0),GravityOverride_15_#5=(X=0,Y=0,Z=-980),JointName_27_#6="root")
    GravityOverrideParams_DBNO_40_#15:Gravity Override Setting Struct Structure=(WindFrequency_28_#2=(X=0,Y=0,Z=0),WindAmplitude_31_#3=(X=0,Y=0,Z=0),WindOffset_32_#4=(X=0,Y=0,Z=0),GravityOverride_15_#5=(X=-980,Y=400,Z=0),JointName_27_#6="pelvis")
    GravityOverrideParams_DBNO_Carried_41_#16:Gravity Override Setting Struct Structure=(WindFrequency_28_#2=(X=0,Y=0,Z=0),WindAmplitude_31_#3=(X=0,Y=0,Z=0),WindOffset_32_#4=(X=0,Y=0,Z=0),GravityOverride_15_#5=(X=-980,Y=500,Z=-500),JointName_27_#6="pelvis")
    GravityOverrideParams_OnGround_Falling_45_#17:Gravity Override Setting Struct Structure=(WindFrequency_28_#2=(X=0,Y=0,Z=0),WindAmplitude_31_#3=(X=0,Y=0,Z=0),WindOffset_32_#4=(X=0,Y=0,Z=0),GravityOverride_15_#5=(X=0,Y=0,Z=-980),JointName_27_#6="root")
    GravityOverrideParams_Ziplining_46_#18:Gravity Override Setting Struct Structure=(WindFrequency_28_#2=(X=0,Y=0,Z=0),WindAmplitude_31_#3=(X=0,Y=0,Z=0),WindOffset_32_#4=(X=0,Y=0,Z=0),GravityOverride_15_#5=(X=0,Y=0,Z=-980),JointName_27_#6="root")
    GravityOverrideParams_Skydive_Surfing_49_#19:Gravity Override Setting Struct Structure=(WindFrequency_28_#2=(X=1.5,Y=7,Z=8),WindAmplitude_31_#3=(X=200,Y=-500,Z=400),WindOffset_32_#4=(X=0,Y=0,Z=0),GravityOverride_15_#5=(X=500,Y=-300,Z=-980),JointName_27_#6="root")
    GravityOverrideParams_Riding_Standing_54_#20:Gravity Override Setting Struct Structure=(WindFrequency_28_#2=(X=0,Y=0,Z=0),WindAmplitude_31_#3=(X=0,Y=0,Z=0),WindOffset_32_#4=(X=0,Y=0,Z=0),GravityOverride_15_#5=(X=0,Y=-350,Z=-980),JointName_27_#6="root")
    GravityOverrideParams_Riding_Moving_55_#21:Gravity Override Setting Struct Structure=(WindFrequency_28_#2=(X=0,Y=0,Z=0),WindAmplitude_31_#3=(X=0,Y=0,Z=0),WindOffset_32_#4=(X=0,Y=0,Z=0),GravityOverride_15_#5=(X=0,Y=-450,Z=-450),JointName_27_#6="root")
    GravityOverrideParams_Riding_Moving_Targeting_58_#22:Gravity Override Setting Struct Structure=(WindFrequency_28_#2=(X=0,Y=0,Z=0),WindAmplitude_31_#3=(X=0,Y=0,Z=0),WindOffset_32_#4=(X=0,Y=0,Z=0),GravityOverride_15_#5=(X=0,Y=-200,Z=-980),JointName_27_#6="root")
    GravityOverrideParams_Tactical_Sprint_57_#23:Gravity Override Setting Struct Structure=(WindFrequency_28_#2=(X=0,Y=0,Z=0),WindAmplitude_31_#3=(X=0,Y=0,Z=0),WindOffset_32_#4=(X=0,Y=0,Z=0),GravityOverride_15_#5=(X=0,Y=-450,Z=-500),JointName_27_#6="root")
    GravityOverrideParams_Rail_Grinding_61_#24:Gravity Override Setting Struct Structure=(WindFrequency_28_#2=(X=0,Y=0,Z=0),WindAmplitude_31_#3=(X=0,Y=0,Z=0),WindOffset_32_#4=(X=0,Y=0,Z=0),GravityOverride_15_#5=(X=0,Y=0,Z=-980),JointName_27_#6="root")
    GravityOverrideParams_DBNO_Tumbling_67_#25:Gravity Override Setting Struct Structure=(WindFrequency_28_#2=(X=0,Y=0,Z=0),WindAmplitude_31_#3=(X=0,Y=0,Z=0),WindOffset_32_#4=(X=0,Y=0,Z=0),GravityOverride_15_#5=(X=-2000,Y=0,Z=0),JointName_27_#6="pelvis")
    GravityOverrideParams_DBNO_RollFwd_68_#26:Gravity Override Setting Struct Structure=(WindFrequency_28_#2=(X=0,Y=0,Z=0),WindAmplitude_31_#3=(X=0,Y=0,Z=0),WindOffset_32_#4=(X=0,Y=0,Z=0),GravityOverride_15_#5=(X=-2000,Y=0,Z=0),JointName_27_#6="pelvis")
    GravityOverrideParams_DBNO_RollBwd_69_#27:Gravity Override Setting Struct Structure=(WindFrequency_28_#2=(X=0,Y=0,Z=0),WindAmplitude_31_#3=(X=0,Y=0,Z=0),WindOffset_32_#4=(X=0,Y=0,Z=0),GravityOverride_15_#5=(X=-4000,Y=0,Z=0),JointName_27_#6="pelvis")
    GravityOverrideParams_DBNO_Impulse_70_#28:Gravity Override Setting Struct Structure=(WindFrequency_28_#2=(X=0,Y=0,Z=0),WindAmplitude_31_#3=(X=0,Y=0,Z=0),WindOffset_32_#4=(X=0,Y=0,Z=0),GravityOverride_15_#5=(X=-2000,Y=0,Z=0),JointName_27_#6="pelvis")
    RBAN Params
    DebugDraw
    DebugJointName
   ) -> (then, GravityOverride, CurrentStateOut, GravityOverrideSettingsOut, RBAN SettingsOut)
   > Set gravityOverride(gravityOverride:Vector=@n3) -> (Output_Get:Vector)
   TryGetPawnOwner()
  ]}
]
variables [
 crouchAlpha {type=number}
 gravityOverride {type=Vector}
]
