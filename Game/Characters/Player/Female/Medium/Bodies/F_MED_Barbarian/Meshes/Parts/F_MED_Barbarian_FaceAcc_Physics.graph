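# Physics asset for the female medium Barbarian's FaceAcc mesh part, defining collision and joint limits for her braided hair.
# Capsules and boxes cover head, neck, upper spine and clavicles, plus center, temple and back braid bones.
# Center, temple and back-B braids form limited-swing, low-twist chains from the head, tuned differently per side in places.
# Body joints, the center-braid root and back-A braids, which attach to the center braid, rotate freely.
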PhysicsAsset /Game/Characters/Player/Female/Medium/Bodies/F_MED_Barbarian/Meshes/Parts/F_MED_Barbarian_FaceAcc_Physics
bodies {
 clavicle_l [
  {shapeType=Capsule, length=7.4607, radius=9.3249, shapeName=backCl, center=<11.2223 -4.1885 -0.1465>, rotation=<rot 0 -89.9999 89.9999>}
  {shapeType=Capsule, length=7.6779, radius=7.5158, shapeName=frontCl, center=<9.3044 2.3587 0.8477>, rotation=<rot 0 90 90>}
 ]
 clavicle_r [
  {shapeType=Capsule, length=7.8979, radius=9.5005, shapeName=backCl, center=<-9.6988 4.1839 0.0583>, rotation=<rot 0 90 89.9999>}
  {shapeType=Capsule, length=7.6779, radius=7.5158, shapeName=frontCl, center=<-9.4125 -2.3587 -0.9849>, rotation=<rot 0 -89.9999 89.9999>}
 ]
 dyn_backBraidA_02_l [{shapeType=Capsule, length=5.4556, radius=1.0482, center=<-3.5078 0.165 -0.3015>, rotation=<rot 0 -93.036 -86.1426>}]
 dyn_backBraidA_02_r [{shapeType=Capsule, length=5.7463, radius=1.0504, center=<3.6219 -0.1894 0.2991>, rotation=<rot 0 -94.6478 -86.6301>}]
 dyn_backBraidB_01_l [{shapeType=Capsule, length=2.9037, radius=1.557, center=<-2.8026 -0.0493 -0.0125>, rotation=<rot 0 92.4072 -86.4373>}]
 dyn_backBraidB_01_r [{shapeType=Capsule, length=2.7338, radius=1.6118, center=<2.2602 0.0615 -0.0152>, rotation=<rot 0 95.304 -85.9805>}]
 dyn_backBraidB_02_l [{shapeType=Capsule, length=3.9535, radius=1.3811, center=<-3.2441 -0.149 -0.0901>, rotation=<rot 0 -92.1168 -89.373>}]
 dyn_backBraidB_02_r [{shapeType=Capsule, length=3.4205, radius=1.6535, center=<2.441 0.1074 0.0891>, rotation=<rot 0 92.2963 -89.1467>}]
 dyn_backBraidB_03_l [{shapeType=Capsule, length=6.2756, radius=1.538, center=<-3.6644 0.1197 0.0017>, rotation=<rot 0 88.0187 -89.7647>}]
 dyn_backBraidB_03_r [{shapeType=Capsule, length=5.7496, radius=1.8597, center=<3.5197 0.1993 -0.0009>, rotation=<rot 0 89.3678 -89.6238>}]
 dyn_backBraidB_Base_l [{shapeType=Capsule, length=3.1217, radius=1.1822, center=<-2.5909 0.9874 -0.4701>, rotation=<rot 0 75.1988 -86.9549>}]
 dyn_backBraidB_Base_r [{shapeType=Capsule, length=1.9769, radius=1.1412, center=<3.1864 -1.2005 0.4817>, rotation=<rot 0 -109.938 -86.2766>}]
 dyn_centerBraid_01 [{shapeType=Capsule, length=2.9579, radius=3.091, center=<2.2331 -0.4351 0>, rotation=<rot 90 0 0>}]
 dyn_centerBraid_02 [{shapeType=Capsule, length=4.2567, radius=3.2402, center=<2.3733 -0.5968 0.0054>, rotation=<rot 82.0906 57.9754 47.1782>}]
 dyn_centerBraid_03 [{shapeType=Capsule, length=3.6079, radius=2.6399, center=<2.1573 0.3729 0>, rotation=<rot 90 178.2101 173.3519>}]
 dyn_centerBraid_Base [{shapeType=Box, extentX=6.1089, extentY=9.609, extentZ=7.6002, center=<2.0204 0 0>, rotation=<rot 0 90 0>}]
 dyn_templeBraid_01_l [{shapeType=Capsule, length=3.2791, radius=0.8541, center=<-2.593 0 -0.1017>, rotation=<rot 0 -89.9999 -92.2434>}]
 dyn_templeBraid_01_r [{shapeType=Capsule, length=2.4967, radius=0.8932, center=<2.3105 0.0455 0.0321>, rotation=<rot 1.6936 -88.6153 -87.5773>}]
 dyn_templeBraid_02_l [{shapeType=Capsule, length=4.2979, radius=0.8584, center=<-2.7357 0.1265 -0.1459>, rotation=<rot -0.0137 -91.6779 -89.0737>}]
 dyn_templeBraid_02_r [{shapeType=Capsule, length=4.038, radius=0.9124, center=<2.4219 -0.0247 -0.0201>, rotation=<rot -0.3595 -89.6085 -89.9169>}]
 dyn_templeBraid_03_l [{shapeType=Capsule, length=2.8089, radius=1.0167, center=<-1.8817 -0.0037 0.0259>, rotation=<rot 0 89.2321 -89.5372>}]
 dyn_templeBraid_03_r [{shapeType=Capsule, length=3.2271, radius=1.0167, center=<1.7164 -0.0451 -0.0272>, rotation=<rot 0 89.2321 -89.5372>}]
 dyn_templeBraid_End_l [{shapeType=Capsule, length=5.0762, radius=1.4978, center=<-3.5473 0.4922 -0.1541>, rotation=<rot 0 85.2249 -87.7576>}]
 dyn_templeBraid_End_r [{shapeType=Capsule, length=5.1098, radius=1.4803, center=<3.5425 -0.4564 0.1526>, rotation=<rot 0.0837 83.086 -87.7591>}]
 head [
  {shapeType=Capsule, length=3.6177, radius=6.3806, center=<5.9704 0.7041 0>, rotation=<rot 0 -54.0348 -89.9998>}
  {shapeType=Capsule, length=7.4823, radius=3.6621, shapeName=head_capsule, center=<-1.9318 0.0001 0>}
 ]
 neck_01 [
  {shapeType=Capsule, length=12.827, radius=5.7065, center=<1.3318 0.4191 0>, rotation=<rot 0 -90 0>}
  {shapeType=Box, extentX=24.5201, extentY=21.3678, extentZ=1.9112, center=<0.017 -10.7942 5.281>, rotation=<rot -3.759 1.0232 -6.3244>}
  {shapeType=Box, extentX=25.6902, extentY=25.939, extentZ=1.6179, center=<0 -8.5226 -5.4525>, rotation=<rot 2.558 0.5669 4.0842>}
 ]
 neck_02 [{shapeType=Capsule, length=9.6339, radius=5.164, center=<0.7865 -0.0002 0.0001>, rotation=<rot 0 -80.4708 89.9999>}]
 spine_05 [
  {shapeType=Capsule, length=21.6226, radius=9.7855, center=<-0.7919 -0.8068 -0.119>}
  {shapeType=Box, extentX=11.1532, extentY=7.5, extentZ=22.9113, center=<8.0567 -3.0791 0.0004>, rotation=<rot 0 43.8245 0>}
 ]
}
constraints [
 {
  bone1Name=head
  bone2Name=neck_02
  swing1LimitDegrees=45
  swing1Motion=Free
  swing2LimitDegrees=45
  swing2Motion=Free
  twistLimitDegrees=45
  twistMotion=Free
 }
 {
  bone1Name=neck_02
  bone2Name=neck_01
  swing1LimitDegrees=45
  swing1Motion=Free
  swing2LimitDegrees=45
  swing2Motion=Free
  twistLimitDegrees=45
  twistMotion=Free
 }
 {
  bone1Name=clavicle_l
  bone2Name=spine_05
  swing1LimitDegrees=45
  swing1Motion=Free
  swing2LimitDegrees=45
  swing2Motion=Free
  twistLimitDegrees=45
  twistMotion=Free
 }
 {
  bone1Name=clavicle_r
  bone2Name=spine_05
  swing1LimitDegrees=45
  swing1Motion=Free
  swing2LimitDegrees=45
  swing2Motion=Free
  twistLimitDegrees=45
  twistMotion=Free
 }
 {bone1Name=dyn_centerBraid_01, bone2Name=dyn_centerBraid_Base, swing1LimitDegrees=45, swing2LimitDegrees=45, twistLimitDegrees=5}
 {bone1Name=dyn_centerBraid_02, bone2Name=dyn_centerBraid_01, swing1LimitDegrees=40, swing2LimitDegrees=40, twistLimitDegrees=5}
 {bone1Name=dyn_centerBraid_03, bone2Name=dyn_centerBraid_02, swing1LimitDegrees=50, swing2LimitDegrees=50, twistLimitDegrees=1}
 {
  bone1Name=neck_01
  bone2Name=spine_05
  swing1LimitDegrees=45
  swing1Motion=Free
  swing2LimitDegrees=45
  swing2Motion=Free
  twistLimitDegrees=45
  twistMotion=Free
 }
 {
  bone1Name=dyn_centerBraid_Base
  bone2Name=head
  swing1LimitDegrees=45
  swing1Motion=Free
  swing2LimitDegrees=45
  swing2Motion=Free
  twistLimitDegrees=45
  twistMotion=Free
 }
 {bone1Name=dyn_templeBraid_01_r, bone2Name=head, swing1LimitDegrees=15, swing2LimitDegrees=15, twistLimitDegrees=1}
 {bone1Name=dyn_templeBraid_02_r, bone2Name=dyn_templeBraid_01_r, swing1LimitDegrees=25, swing2LimitDegrees=25, twistLimitDegrees=2}
 {bone1Name=dyn_templeBraid_03_r, bone2Name=dyn_templeBraid_02_r, swing1LimitDegrees=25, swing2LimitDegrees=25, twistLimitDegrees=2}
 {bone1Name=dyn_templeBraid_End_r, bone2Name=dyn_templeBraid_03_r, swing1LimitDegrees=25, swing2LimitDegrees=25, twistLimitDegrees=2}
 {bone1Name=dyn_templeBraid_01_l, bone2Name=head, swing1LimitDegrees=20, swing2LimitDegrees=20, twistLimitDegrees=2}
 {bone1Name=dyn_templeBraid_02_l, bone2Name=dyn_templeBraid_01_l, swing1LimitDegrees=20, swing2LimitDegrees=20, twistLimitDegrees=2}
 {bone1Name=dyn_templeBraid_03_l, bone2Name=dyn_templeBraid_02_l, swing1LimitDegrees=20, swing2LimitDegrees=20, twistLimitDegrees=2}
 {bone1Name=dyn_templeBraid_End_l, bone2Name=dyn_templeBraid_03_l, swing1LimitDegrees=20, swing2LimitDegrees=20, twistLimitDegrees=2}
 {bone1Name=dyn_backBraidB_01_l, bone2Name=dyn_backBraidB_Base_l, swing1LimitDegrees=30, swing2LimitDegrees=10, twistLimitDegrees=10}
 {bone1Name=dyn_backBraidB_02_l, bone2Name=dyn_backBraidB_01_l, swing1LimitDegrees=20, swing2LimitDegrees=5, twistLimitDegrees=5}
 {bone1Name=dyn_backBraidB_03_l, bone2Name=dyn_backBraidB_02_l, swing1LimitDegrees=20, swing2LimitDegrees=5, twistLimitDegrees=5}
 {bone1Name=dyn_backBraidB_02_r, bone2Name=dyn_backBraidB_01_r, swing1LimitDegrees=20, swing2LimitDegrees=20, twistLimitDegrees=10}
 {bone1Name=dyn_backBraidB_03_r, bone2Name=dyn_backBraidB_02_r, swing1LimitDegrees=20, swing2LimitDegrees=20, twistLimitDegrees=10}
 {
  bone1Name=dyn_backBraidA_02_l
  bone2Name=dyn_centerBraid_01
  swing1LimitDegrees=45
  swing1Motion=Free
  swing2LimitDegrees=45
  swing2Motion=Free
  twistLimitDegrees=45
  twistMotion=Free
 }
 {
  bone1Name=dyn_backBraidA_02_r
  bone2Name=dyn_centerBraid_01
  swing1LimitDegrees=45
  swing1Motion=Free
  swing2LimitDegrees=45
  swing2Motion=Free
  twistLimitDegrees=45
  twistMotion=Free
 }
 {bone1Name=dyn_backBraidB_01_r, bone2Name=dyn_backBraidB_Base_r, swing1LimitDegrees=20, swing2LimitDegrees=20, twistLimitDegrees=10}
 {bone1Name=dyn_backBraidB_Base_r, bone2Name=head, swing1LimitDegrees=45, swing2LimitDegrees=45, twistLimitDegrees=45}
 {bone1Name=dyn_backBraidB_Base_l, bone2Name=head, swing1LimitDegrees=45, swing2LimitDegrees=45, twistLimitDegrees=10}
]
