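# Post-process style hair AnimBP for the female medium Barbarian's face accessory mesh, simulating braid motion.
# Copies the parent mesh pose, offsets the center braid by head yaw, skydiving and parachute state, then caches it.
# Blends frontend or gameplay RigidBody sims, both pushed by Perlin-noise falling force during skydives, damped under parachute.
# Head-driven rotation and half-weight constraints to the center braid then adjust the back braids.
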AnimBlueprint /Game/Characters/Player/Female/Medium/Bodies/F_MED_Barbarian/Meshes/Parts/F_MED_Barbarian_FaceAcc_AnimBP
target_skeleton=&/F_MED_Barbarian_FaceAcc_hair_MESH_Skeleton
graphs [
 AnimGraph {nodes [
   n1 = RigidBody(ComponentPose=Local To Component(LocalPose=Use cached pose 'hairPose'()), ExternalForce=Get fallingForce() -> (fallingForce:Vector)) {
    bClampLinearTranslationLimitToRefPose=true
    bOverrideWorldGravity=true
    overridePhysicsAsset=&/F_MED_Barbarian_FaceAcc_FrontendPhysics
    simulationSpace=BaseBoneSpace
    alphaBoolBlend {blendOption=Linear}
    alphaScaleBiasClamp {bInterpResult=true, interpSpeedDecreasing=0.5, interpSpeedIncreasing=0.5}
    baseBoneRef {boneName=spine_05}
    componentAppliedLinearAccClamp=<1000 4000 1000>
    componentLinearAccScale=<0.7 0.7 0.7>
    componentLinearVelScale=<0.6 0.6 0.6>
    overrideWorldGravity {z=-980}
   }
   n2 = RigidBody(ComponentPose=Local To Component(LocalPose=Use cached pose 'hairPose'()), ExternalForce=Get fallingForce()) {
    bClampLinearTranslationLimitToRefPose=true
    bOverrideWorldGravity=true
    overridePhysicsAsset=&/F_MED_Barbarian_FaceAcc_Physics
    simulationSpace=BaseBoneSpace
    alphaBoolBlend {blendOption=Linear}
    alphaScaleBiasClamp {bInterpResult=true, interpSpeedDecreasing=0.5, interpSpeedIncreasing=0.5}
    baseBoneRef {boneName=spine_05}
    componentAppliedLinearAccClamp=<1000 4000 1000>
    componentLinearAccScale=<0.7 0.7 0.7>
    componentLinearVelScale=<0.6 0.6 0.6>
    overrideWorldGravity {z=-980}
   }
   n3 = Blend Poses by bool(
    BlendPose_0=Component To Local(ComponentPose=@n1)
    BlendPose_1=Component To Local(ComponentPose=@n2)
    BlendTime_1=0.5
    bActiveValue=Get bIsInFrontEnd()
   ) {blendType=Linear}
   n4 = Get braidYaw() -> (braidYaw:Rotator)
   n5 = "Transform (Modify) Bone - Bone: dyn_backBraidB_Base_l"(
    ComponentPose="Transform (Modify) Bone - Bone: dyn_backBraidB_Base_r"(ComponentPose=Local To Component(LocalPose=@n3), Alpha=0.3, Translation="0,0,0", Rotation=@n4, Scale="1,1,1") {rotationMode=BMM_Additive, alphaBoolBlend {blendOption=Linear}, boneToModify {boneName=dyn_backBraidB_Base_r}}
    Alpha=0.3
    Translation="0,0,0"
    Rotation=@n4
    Scale="1,1,1"
   ) {rotationMode=BMM_Additive, rotationSpace=BCS_BoneSpace, alphaBoolBlend {blendOption=Linear}, boneToModify {boneName=dyn_backBraidB_Base_l}}
   n6 = "Constraint - dyn_backBraidA_02_r (P)"(ComponentPose=@n5, Alpha=0.5, ConstraintWeights_0=1) {
    alphaBoolBlend {blendOption=Linear}
    alphaScaleBiasClamp {bInterpResult=true, interpSpeedDecreasing=0.5, interpSpeedIncreasing=0.5}
    boneToModify {boneName=dyn_backBraidA_02_r}
    constraintSetup [{targetBone {boneName=dyn_centerBraid_03}}]
    constraintWeights [1]
   }
   n7 = "Constraint - dyn_backBraidA_02_l (P)"(ComponentPose=@n6, Alpha=0.5, ConstraintWeights_0=1) {
    alphaBoolBlend {blendOption=Linear}
    alphaScaleBiasClamp {bInterpResult=true, interpSpeedDecreasing=2, interpSpeedIncreasing=2}
    boneToModify {boneName=dyn_backBraidA_02_l}
    constraintSetup [{targetBone {boneName=dyn_centerBraid_03}}]
    constraintWeights [1]
   }
   n8 = "Constraint - dyn_backBraidA_01_r (P)"(ComponentPose=@n7, Alpha=0.5, ConstraintWeights_0=1) {
    alphaBoolBlend {blendOption=Linear}
    alphaScaleBiasClamp {bInterpResult=true, interpSpeedDecreasing=0.5, interpSpeedIncreasing=0.5}
    boneToModify {boneName=dyn_backBraidA_01_r}
    constraintSetup [{targetBone {boneName=dyn_centerBraid_02}}]
    constraintWeights [1]
   }
   n9 = "Constraint - dyn_backBraidA_01_l (P)"(ComponentPose=@n8, Alpha=0.5, ConstraintWeights_0=1) {
    alphaBoolBlend {blendOption=Linear}
    alphaScaleBiasClamp {bInterpResult=true, interpSpeedDecreasing=2, interpSpeedIncreasing=2}
    boneToModify {boneName=dyn_backBraidA_01_l}
    constraintSetup [{targetBone {boneName=dyn_centerBraid_02}}]
    constraintWeights [1]
   }
   Output Pose(Result=Component To Local(ComponentPose=@n9)) {layerGroup=None}
   n10 = "Transform (Modify) Bone - Bone: dyn_centerBraid_Base"(
    ComponentPose=Local To Component(LocalPose=Copy Pose From Mesh() {bUseAttachedParent=true})
    Alpha=1
    Translation="0,0,0"
    Rotation=Get headYaw() -> (headYaw:Rotator)
    Scale="1,1,1"
   ) {
    rotationMode=BMM_Additive
    rotationSpace=BCS_BoneSpace
    translationMode=BMM_Additive
    alphaBoolBlend {blendOption=Linear}
    boneToModify {boneName=dyn_centerBraid_Base}
   }
   Save cached pose 'hairPose'(Pose=Component To Local(ComponentPose=@n10)) {cacheName=hairPose}
  ]}
 EventGraph {nodes [
   Event BlueprintUpdateAnimation()
   n1 = BreakTransform(InTransform=GetDeltaTransformFromRefPose(self=GetOwnerSkeletalMeshComponent(), BoneName=head, BaseName=spine_05))
   n2 = Get bIsSkyDiving()
   n3 = "To Float (Boolean)"(InBool=@n2)
   n4 = float - float(
    A=MapRangeClamped(Value=BreakRotator(InRot=@n1.Rotation).Yaw, InRangeA=26, InRangeB=-43, OutRangeA=5, OutRangeB=-50)
    B=float * float(A=22, B=@n3)
   ) {is=title K2Node_CallFunction}
   n5 = Get bIsParachuteOpened()
   n6 = float * float(A=10, B="To Float (Boolean)"(InBool=@n5))
   > Set headYaw(headYaw:Rotator=MakeRotator(Yaw=float + float(A=@n4, B=@n6) {is=title K2Node_CommutativeAssociativeBinaryOperator})) -> (Output_Get:Rotator)
   n7 = "To Float (Boolean)"(InBool=Get bIsTargeting())
   > Set ADS(ADS:double=float - float(B=float * float(A=@n7, B=0.4))) -> (Output_Get:double)
   n8 = GetDeltaTransformFromRefPose(self=GetOwnerSkeletalMeshComponent(), BoneName=head, BaseName=spine_05)
   n9 = MapRangeClamped(Value=BreakRotator(InRot=BreakTransform(InTransform=@n8).Rotation).Yaw, InRangeA=26, InRangeB=-43, OutRangeA=5, OutRangeB=-60)
   > Set braidYaw(braidYaw:Rotator=MakeRotator(Roll=@n9, Yaw=@n9)) -> (Output_Get:Rotator)
   n10 = Get PawnLastLocation() -> (PawnLastLocation:Vector)
   n11 = SelectFloat(
    A=Make Perlin Noise and Remap!(Value:float=BreakVector(InVec=@n10).Z, RangeOutMin:float=-15000, RangeOutMax:float=15000) -> (ReturnValue:float)
    B=0
    bPickA=@n2
   )
   n12 = SelectFloat(A=float * float(A=@n11, B=0.2), B=@n11, bPickA=@n5)
   > Set fallingForce(fallingForce:Vector=MakeVector(X=@n12, Y=@n12)) -> (Output_Get:Vector)
   float - float(A=@n9, B=float * float(A=15, B=@n3))
  ]}
]
variables [
 ADS {type=number}
 braidYaw {type=Rotator}
 fallingForce {type=Vector}
 headYaw {type=Rotator}
]
